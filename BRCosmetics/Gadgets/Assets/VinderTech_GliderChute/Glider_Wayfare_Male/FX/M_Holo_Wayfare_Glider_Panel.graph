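# Unlit material for a holographic panel on the male Wayfare glider, showing animated scanlines in a background and line color.
# Two samples of one scanline texture scroll along V on UV1 at different tilings, while a third has its UVs scaled by a sine of time.
# Their weighted sum blends the two color parameters, feeding emissive through MF_FX_EmissiveAdjustments.
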Material /BRCosmetics/Gadgets/Assets/VinderTech_GliderChute/Glider_Wayfare_Male/FX/M_Holo_Wayfare_Glider_Panel
graph {
 properties {
  allowTranslucentCustomDepthWrites=true
  bUsedWithMeshDeformer=true
  bUsedWithMorphTargets=true
  bUsedWithNiagaraMeshParticles=true
  bUsedWithSkeletalMesh=true
  pixelDepthOffsetMode=PDOM_Legacy
  shadingModel=MSM_Unlit
  usedShadingModels=Unlit
 }
 nodes [
  n1 = TextureCoordinate() {coordinateIndex=1}
  n2 = Panner(Coordinate=@n1) {speedY=0.5}
  n3 = Panner(Coordinate=Multiply(A=@n1) {constB=0.2}) {speedY=0.5}
  n4 = Add(
   A=Multiply(A=TextureSample(UVs=@n2) {texture=/Game/Global/Textures/Default/Masks_and_Gradients/Scanlines_Offset}.R) {constB=0.2}
   B=Multiply(A=TextureSample(UVs=@n3) {texture=&Scanlines_Offset}.R)
  )
  n5 = Multiply(A=@n1, B=Sine(None=Time()))
  n6 = Multiply(A=TextureSample(UVs=Panner(Coordinate=@n5)) {texture=&Scanlines_Offset}.R) {constB=0.3}
  MaterialFunctionCall(Emissive=LinearInterpolate(
    A=VectorParameter() {parameterName=M_Background_Color, defaultValue=<rgba 0 0.0322 0.3333 1>}.RGB
    B=VectorParameter() {parameterName=M_Lines_Color, defaultValue=<rgba 0.4177 0.7422 1 1>}.RGB
    Alpha=Add(A=@n4, B=@n6)
   )) {materialFunction=&MF_FX_EmissiveAdjustments}
 ]
}
